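# Control Rig for the Tower Sentinel Boss face accessory that procedurally rotates headdress bones from head motion.
# It converts the head bone's local rotation to Euler angles and feeds clamped remaps into small additive rotations.
# Head X drives four bones in each left and right chain over opposing ranges, while head Z rotates both master bones.
# The headdress aim bones in the hierarchy are unused by the graph.
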ControlRigBlueprint /BRCosmetics/Characters/Player/Female/Medium/Bodies/F_MED_Tower_Sentinel_Boss/Meshes/Parts/F_MED_Tower_Sentinel_Boss_A_FaceAcc_CtrlRig
graphs [
 RigVMFunctionLibrary {links [], nodes []}
 RigVMModel {nodes [
   RigUnit_BeginExecution = &RigUnit_BeginExecution()
   n1 = &RigVMFunction_MathQuaternionToEuler(RotationOrder=ZYX, Value=&RigUnit_GetTransform(Item=(Type=Bone,Name="head"), Space=LocalSpace).Transform.Rotation)
   n2 = &RigVMFunction_MathQuaternionFromEuler(
    Euler=(X=0,Y=0,Z=0)
    Euler.Z=&RigVMFunction_MathFloatRemap(SourceMaximum=-15, SourceMinimum=10, TargetMaximum=5, TargetMinimum=0, bClamp=true, Value=@n1.Result.X).Result
    Euler.Y=&RigVMFunction_MathFloatRemap(SourceMaximum=-20, SourceMinimum=0, TargetMaximum=3, TargetMinimum=0, bClamp=true, Value=@n1.Result.X).Result
   )
   &RigUnit_ModifyTransforms(
    ItemToModify="((Item=(Type=Bone,Name=\"dyn_headdress_l_01\"),Transform=(Rotation=(X=0,Y=0,Z=0,W=1),Translation=(X=0,Y=0,Z=0),Scale3D=(X=1,Y=1,Z=1))),(Item=(Type=Bone,Name=\"dyn_headdress_l_02\"),Transform=(Rotation=(X=0,Y=0,Z=0,W=1),Translation=(X=0,Y=0,Z=0),Scale3D=(X=1,Y= ...(truncated)"
    Weight=1
    ItemToModify.0.Transform.Rotation=@n2.Result
    ItemToModify.1.Transform.Rotation=@n2.Result
    ItemToModify.2.Transform.Rotation=@n2.Result
    ItemToModify.3.Transform.Rotation=@n2.Result
    ExecutePin=@RigUnit_BeginExecution.ExecutePin
   )
   n3 = &RigUnit_GetTransform(Item=(Type=Bone,Name="head"), Space=LocalSpace)
   n4 = &RigVMFunction_MathQuaternionToEuler(RotationOrder=ZYX, Value=@n3.Transform.Rotation)
   n5 = &RigVMFunction_MathQuaternionFromEuler(
    Euler=(X=0,Y=0,Z=0)
    Euler.Z=&RigVMFunction_MathFloatRemap(SourceMaximum=25, SourceMinimum=-5, TargetMaximum=-3, TargetMinimum=0, bClamp=true, Value=@n4.Result.X).Result
    Euler.Y=&RigVMFunction_MathFloatRemap(SourceMaximum=25, SourceMinimum=-5, TargetMaximum=-4, TargetMinimum=0, bClamp=true, Value=@n4.Result.X).Result
   )
   > &RigUnit_ModifyTransforms(
    ItemToModify="((Item=(Type=Bone,Name=\"dyn_headdress_r_01\"),Transform=(Rotation=(X=0,Y=0,Z=0,W=1),Translation=(X=0,Y=0,Z=0),Scale3D=(X=1,Y=1,Z=1))),(Item=(Type=Bone,Name=\"dyn_headdress_r_02\"),Transform=(Rotation=(X=0,Y=0,Z=0,W=1),Translation=(X=0,Y=0,Z=0),Scale3D=(X=1,Y= ...(truncated)"
    Weight=1
    ItemToModify.0.Transform.Rotation=@n5.Result
    ItemToModify.1.Transform.Rotation=@n5.Result
    ItemToModify.2.Transform.Rotation=@n5.Result
    ItemToModify.3.Transform.Rotation=@n5.Result
   )
   n6 = &RigVMFunction_MathQuaternionFromEuler(
    Euler=(X=0,Y=0,Z=0)
    Euler.Y=&RigVMFunction_MathFloatRemap(
     SourceMaximum=5
     SourceMinimum=-10
     TargetMaximum=-7
     TargetMinimum=0
     bClamp=true
     Value=&RigVMFunction_MathQuaternionToEuler(RotationOrder=ZYX, Value=@n3.Transform.Rotation).Result.Z
    ).Result
   )
   > &RigUnit_ModifyTransforms(
    ItemToModify="((Item=(Type=Bone,Name=\"dyn_headdress_r_MSTR\"),Transform=(Rotation=(X=0,Y=0,Z=0,W=1),Translation=(X=0,Y=0,Z=0),Scale3D=(X=1,Y=1,Z=1))),(Item=(Type=Bone,Name=\"dyn_headdress_l_MSTR\"),Transform=(Rotation=(X=0,Y=0,Z=0,W=1),Translation=(X=0,Y=0,Z=0),Scale3D=(X= ...(truncated)"
    Weight=1
    ItemToModify.0.Transform.Rotation=@n6.Result
    ItemToModify.1.Transform.Rotation=@n6.Result
   )
  ]}
]
hierarchy [
 root {loc=<0 0 0>, rot=<0 0 0 1>}
 pelvis {parent=root, loc=<0 -0.5301 98.0571>, rot=<0.0127 0.707 -0.0127 -0.707>}
 spine_01 {parent=pelvis, loc=<6.1208 -0.9328 0>, rot=<0 0 0.034 -0.9994>}
 spine_02 {parent=spine_01, loc=<6.5996 -0.2095 0>, rot=<0 0 0.0238 -0.9997>}
 spine_03 {parent=spine_02, loc=<6.4914 -0.1237 0>, rot=<0 0 0.0149 -0.9999>}
 spine_04 {parent=spine_03, loc=<6.9598 0.0794 0>, rot=<0 0 -0.0132 -0.9999>}
 spine_05 {parent=spine_04, loc=<6.9987 0 0>, rot=<0 0 -0.0015 -1>}
 clavicle_l {parent=spine_05, loc=<5.4831 -0.1053 -3.2946>, rot=<0.0465 -0.7289 -0.009 -0.6829>}
 clavicle_r {parent=spine_05, loc=<5.4831 -0.1053 3.2946>, rot=<0.7289 0.0465 -0.6829 0.009>}
 neck_01 {parent=spine_05, loc=<11.3138 0.1529 0>, rot=<0 0 -0.2003 -0.9797>}
 neck_02 {parent=neck_01, loc=<4.5627 0 0>, rot=<0 0 0.0504 -0.9987>}
 head {parent=neck_02, loc=<5.3914 0 0>, rot=<0 0 0.115 -0.9934>}
 dyn_headdress_r_MSTR {parent=head, loc=<2.2202 -4.4019 7.3143>, rot=<-0.0029 0.7071 0.7071 -0.0029>}
 dyn_headdress_r_01 {parent=dyn_headdress_r_MSTR, loc=<0.4164 0 0>, rot=<0.1717 0.0367 -0.0064 -0.9844>}
 dyn_headdress_r_02 {parent=dyn_headdress_r_01, loc=<4.9068 -0.367 -0.366>, rot=<-0.1959 0.2588 0.0698 -0.9433>}
 dyn_headdress_r_03 {parent=dyn_headdress_r_02, loc=<4.5717 0 0>, rot=<-0.0093 0.1526 0.0089 -0.9882>}
 dyn_headdress_r_04 {parent=dyn_headdress_r_03, loc=<5.0567 0 0>, rot=<-0.0142 0.0133 0.019 -0.9996>}
 dyn_headdress_r_05 {parent=dyn_headdress_r_04, loc=<5.1646 0 0>, rot=<0.1055 -0.2113 -0.094 -0.9672>}
 dyn_headdress_l_MSTR {parent=head, loc=<2.2202 -4.4019 -7.3143>, rot=<-0.0029 0.7071 0.7071 -0.0029>}
 dyn_headdress_l_01 {parent=dyn_headdress_l_MSTR, loc=<0.4164 0 0>, rot=<-0.1717 0.0064 -0.0367 -0.9844>}
 dyn_headdress_l_02 {parent=dyn_headdress_l_01, loc=<4.9341 0 0>, rot=<0.2092 0.2861 -0.0277 -0.9347>}
 dyn_headdress_l_03 {parent=dyn_headdress_l_02, loc=<4.5717 0 0>, rot=<0.0093 0.1526 -0.0089 -0.9882>}
 dyn_headdress_l_04 {parent=dyn_headdress_l_03, loc=<5.0567 0 0>, rot=<0.0142 0.0133 -0.019 -0.9996>}
 dyn_headdress_l_05 {parent=dyn_headdress_l_04, loc=<5.1646 0 0>, rot=<-0.1055 -0.2113 0.094 -0.9672>}
 dyn_headdress_r_aim {parent=spine_04, loc=<12.4655 9.0494 4.0485>, rot=<0 0 -0.0015 -1>}
 dyn_headdress_l_aim {parent=spine_04, loc=<12.4655 9.0494 -4.049>, rot=<0 0 -0.0015 -1>}
]
