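# Physics asset for the Tower Sentinel Boss face accessory (female medium cosmetic body), defining physics for its left and right headdress pieces.
# Each side is a mirrored five-capsule chain rooted at a small MSTR body, with tight angular limits that mostly loosen toward the tip.
# Capsules on the neck and upper spine supply body volumes, linked to each other and to both MSTR roots by angularly free constraints.
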PhysicsAsset /BRCosmetics/Characters/Player/Female/Medium/Bodies/F_MED_Tower_Sentinel_Boss/Meshes/Parts/F_MED_Tower_Sentinel_Boss_A_FaceAcc_Physics
bodies {
 dyn_headdress_l_01 [{shapeType=Capsule, length=3.5, radius=3, center=<2 0 0>, rotation=<rot 90 0 0>}]
 dyn_headdress_l_02 [{shapeType=Capsule, length=6.5557, radius=2.3445, center=<2.2 0 0>, rotation=<rot 90 0 0>}]
 dyn_headdress_l_03 [{shapeType=Capsule, length=3.6634, radius=2.1411, center=<2.5575 0 0>, rotation=<rot 90 0 0>}]
 dyn_headdress_l_04 [{shapeType=Capsule, length=3.402, radius=1.9616, center=<2.5641 0 0>, rotation=<rot 90 0 0>}]
 dyn_headdress_l_05 [{shapeType=Capsule, length=4.4975, radius=1.9123, center=<4.3347 0 0>, rotation=<rot 90 0 0>}]
 dyn_headdress_l_MSTR [{shapeType=Capsule, length=0.505, radius=0.505}]
 dyn_headdress_r_01 [{shapeType=Capsule, length=3.5, radius=3, center=<2 0 0>, rotation=<rot 90 0 0>}]
 dyn_headdress_r_02 [{shapeType=Capsule, length=6.5557, radius=2.3445, center=<2.2 0 0>, rotation=<rot 90 0 0>}]
 dyn_headdress_r_03 [{shapeType=Capsule, length=3.6634, radius=2.1411, center=<2.5575 0 0>, rotation=<rot 90 0 0>}]
 dyn_headdress_r_04 [{shapeType=Capsule, length=3.402, radius=1.9616, center=<2.5641 0 0>, rotation=<rot 90 0 0>}]
 dyn_headdress_r_05 [{shapeType=Capsule, length=4.4975, radius=1.9123, center=<4.3347 0 0>, rotation=<rot 90 0 0>}]
 dyn_headdress_r_MSTR [{shapeType=Capsule, length=0.505, radius=0.505}]
 neck_01 [{shapeType=Capsule, length=10.1042, radius=4.8707, center=<5.0144 0.0003 0>, rotation=<rot 90 0 0>}]
 spine_04 [{shapeType=Capsule, length=19.1825, radius=11.6737, center=<0.615 -0.0667 0>}]
 spine_05 [{shapeType=Capsule, length=19.2923, radius=9.3778, center=<1.3693 -2.6946 0>, rotation=<rot 0 -15.0016 0>}]
}
constraints [
 {bone1Name=dyn_headdress_l_05, bone2Name=dyn_headdress_l_04, swing1LimitDegrees=5, swing2LimitDegrees=15, twistLimitDegrees=3}
 {bone1Name=dyn_headdress_l_02, bone2Name=dyn_headdress_l_01, swing1LimitDegrees=3, swing2LimitDegrees=10, twistLimitDegrees=3}
 {bone1Name=dyn_headdress_l_03, bone2Name=dyn_headdress_l_02, swing1LimitDegrees=5, swing2LimitDegrees=10, twistLimitDegrees=3}
 {bone1Name=dyn_headdress_l_04, bone2Name=dyn_headdress_l_03, swing1LimitDegrees=5, swing2LimitDegrees=15, twistLimitDegrees=3}
 {bone1Name=dyn_headdress_r_05, bone2Name=dyn_headdress_r_04, swing1LimitDegrees=5, swing2LimitDegrees=15, twistLimitDegrees=3}
 {bone1Name=dyn_headdress_r_02, bone2Name=dyn_headdress_r_01, swing1LimitDegrees=3, swing2LimitDegrees=10, twistLimitDegrees=3}
 {bone1Name=dyn_headdress_r_03, bone2Name=dyn_headdress_r_02, swing1LimitDegrees=5, swing2LimitDegrees=10, twistLimitDegrees=3}
 {bone1Name=dyn_headdress_r_04, bone2Name=dyn_headdress_r_03, swing1LimitDegrees=5, swing2LimitDegrees=15, twistLimitDegrees=3}
 {bone1Name=dyn_headdress_r_01, bone2Name=dyn_headdress_r_MSTR, swing1LimitDegrees=3, swing2LimitDegrees=3, twistLimitDegrees=3}
 {bone1Name=dyn_headdress_l_01, bone2Name=dyn_headdress_l_MSTR, swing1LimitDegrees=3, swing2LimitDegrees=3, twistLimitDegrees=3}
 {
  bone1Name=dyn_headdress_r_MSTR
  bone2Name=neck_01
  swing1LimitDegrees=45
  swing1Motion=Free
  swing2LimitDegrees=45
  swing2Motion=Free
  twistLimitDegrees=45
  twistMotion=Free
 }
 {
  bone1Name=dyn_headdress_l_MSTR
  bone2Name=neck_01
  swing1LimitDegrees=45
  swing1Motion=Free
  swing2LimitDegrees=45
  swing2Motion=Free
  twistLimitDegrees=45
  twistMotion=Free
 }
 {
  bone1Name=spine_05
  bone2Name=spine_04
  swing1LimitDegrees=45
  swing1Motion=Free
  swing2LimitDegrees=45
  swing2Motion=Free
  twistLimitDegrees=45
  twistMotion=Free
 }
 {
  bone1Name=neck_01
  bone2Name=spine_05
  swing1LimitDegrees=45
  swing1Motion=Free
  swing2LimitDegrees=45
  swing2Motion=Free
  twistLimitDegrees=45
  twistMotion=Free
 }
]
