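# Animation Blueprint for the Tower Sentinel Boss female face accessory mesh, whose output pose is only Copy Pose From Mesh using the attached parent.
# A head-space RigidBody feeding a Control Rig, plus two headdress Look At nodes, sit in the graph unconnected to the output.
# The event graph caches pawn, body and head meshes, then maps head yaw relative to spine_05 into headUp, which nothing here reads.
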AnimBlueprint /BRCosmetics/Characters/Player/Female/Medium/Bodies/F_MED_Tower_Sentinel_Boss/Meshes/Parts/F_MED_Tower_Sentinel_Boss_A_FaceAcc_AnimBP
target_skeleton=&/F_MED_Tower_Sentinel_Boss_A_FaceAcc_Skeleton
graphs [
 AnimGraph {nodes [
   n1 = RigidBody(ComponentPose=Local To Component()) {
    bOverrideWorldGravity=true
    overridePhysicsAsset=&/F_MED_Tower_Sentinel_Boss_A_FaceAcc_Physics
    simulationSpace=BaseBoneSpace
    alphaBoolBlend {blendOption=Linear}
    baseBoneRef {boneName=head}
    componentAppliedLinearAccClamp=<300 300 300>
    componentLinearAccScale=<0.3 0.3 0.5>
    componentLinearVelScale=<0.3 0.3 0.3>
    overrideWorldGravity {z=-980}
   }
   Control Rig(Source=Component To Local(ComponentPose=@n1)) {
    alphaBoolBlend {blendOption=Linear}
    controlRigAssetReference {blueprintRigClass=&/F_MED_Tower_Sentinel_Boss_A_FaceAcc_CtrlRig.F_MED_Tower_Sentinel_Boss_A_FaceAcc_CtrlRig_C}
   }
   Look At - Bone: dyn_headdress_l_MSTR(Alpha=1) {
    lookAtClamp=25
    alphaBoolBlend {blendOption=Linear}
    boneToModify {boneName=dyn_headdress_l_MSTR}
    lookAtLocation {x=0}
    lookAtTarget {boneReference {boneName=dyn_headdress_l_aim}}
    lookAt_Axis {axis=<1 0>}
    lookUp_Axis {axis {x=0}}
   }
   Look At - Bone: dyn_headdress_r_01(Alpha=1) {
    alphaBoolBlend {blendOption=Linear}
    boneToModify {boneName=dyn_headdress_r_01}
    lookAtLocation=<0 8 -7>
    lookAtTarget {boneReference {boneName=dyn_headdress_r_aim}}
    lookAt_Axis {axis=<1 0>}
    lookUp_Axis {axis {x=0}}
   }
   Output Pose(Result=Copy Pose From Mesh() {bUseAttachedParent=true}) {layerGroup=None}
  ]}
 EventGraph {nodes [
   Event BlueprintInitializeAnimation()
   > Set Body_Mesh(Body_Mesh=GetOwnerSkeletalMeshComponent())
   > n1 = Cast To PlayerPawn_Athena(Object=TryGetPawnOwner())
   > n2 = Set Pawn Owner(Pawn Owner:Player Pawn Athena Object Reference=@n1) -> (Output_Get:Player Pawn Athena Object Reference)
   > Set Head_Mesh(Head_Mesh:Skeletal Mesh Component Object Reference=GetSkeletalMeshForPartType(self=@n2, PartType=Head)) -> (Output_Get:Skeletal Mesh Component Object Reference)
   Event BlueprintUpdateAnimation()
   > Sequence()
   n3 = BreakTransform(InTransform=GetDeltaTransformFromRefPose(self=GetOwnerSkeletalMeshComponent(), BoneName=head, BaseName=spine_05))
   > Set headUp(headUp:double=MapRangeClamped(Value=BreakRotator(InRot=@n3.Rotation).Yaw, InRangeA=0, InRangeB=-50)) -> (Output_Get:double)
  ]}
]
variables [
 Body_Mesh {type=SkeletalMeshComponent}
 Head_Mesh {type=SkeletalMeshComponent}
 headUp {type=number}
 Pawn Owner {type=PlayerPawn_Athena_C}
]
